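# Masked, lit particle material for paint-drip effects on the Graffiti pickaxe, usable on mesh particles and Niagara sprites.
# A dynamic-parameter frame selects splash shapes from a SubUV flipbook, whose UVs are distorted by panning cloud noise masked through a plasma flipbook.
# Particle color tints a blend of desaturated camera-vector cubemap and splash channel, plus faint emissive.
# Splash alpha, dynamic erosion and particle alpha shape the cutout; a scrolling cloud normal adds detail.
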Material /BRCosmetics/Weapons/FORT_Melee/Pickaxe_Graffiti/Effects/M_Melee_Pickaxe_Graffiti_Drips
graph {
 properties {
  bUsedWithMeshParticles=true
  bUsedWithNiagaraMeshParticles=true
  bUsedWithNiagaraSprites=true
  blendMode=BLEND_Masked
  pixelDepthOffsetMode=PDOM_Legacy
  usedShadingModels=Default Lit
 }
 nodes [
  n1 = ScalarParameter() {defaultValue=4, parameterName="U_Images (S)"}
  n2 = ScalarParameter() {defaultValue=3, parameterName="V_Images (S)"}
  n3 = TextureCoordinate() {uTiling=0.5, vTiling=0.5}
  n4 = DynamicParameter() {defaultValue=<rgba 0 0 0 0.5>, paramNames [UVDistortScale, UvTxtureScale, RandomFrame, Erode]}
  n5 = TextureSample(UVs=Panner(Coordinate=Multiply(A=@n3, B=@n4.UvTxtureScale)) {speedY=0.5}) {texture=/Game/Effects/Fort_Effects/Textures/Tile/T_CloudScape_08}
  n6 = DynamicParameter() {defaultValue=<rgba 0 0 0 0.5>, paramNames [UVDistortScale, UvTxtureScale, RandomFrame, Erode]}
  n7 = StaticBoolParameter() {defaultValue=true, parameterName="Manual Time Control (B)"}
  n8 = MaterialFunctionCall!(
   U_Images=@n1
   V_Images=@n2
   SubUV Texture RGB=TextureObject() {texture=/Game/Effects/Fort_Effects/Textures/Plasma/T_PlasmaUV_Mask}
   Time Speed
   Manual Time Control=@n7
   Time=DynamicParameter() {defaultValue=<rgba 0 0 0 0.5>, paramNames [UVDistortScale, UvTxtureScale, RandomFrame, Erode]}.RandomFrame
   UVs
  ) {materialFunction=/Game/Effects/Fort_Effects/Materials/_Functions/MF_SubUV}
  n9 = Add(A=TextureCoordinate(), B=Multiply(A=Multiply(A=@n5.G, B=@n6.UVDistortScale), B=ComponentMask(None=@n8.RGB) {g=true}))
  n10 = ParticleMacroUV()
  n11 = Panner(Coordinate=TextureCoordinate()) {speedY=0.1}
  Add(
   A=MaterialFunctionCall(
    U_Images=@n1
    V_Images=@n2
    SubUV Texture RGB=TextureObject() {texture=/Game/Effects/Fort_Effects/Textures/Liquid/T_PoisonSplashShapes_6X6_Packed}
    Manual Time Control=StaticBoolParameter() {defaultValue=true, parameterName="Manual Time Control (B)_1"}
    Time=DynamicParameter() {defaultValue=<rgba 0 0 0 0.5>, paramNames [UVDistortScale, UvTxtureScale, RandomFrame, Erode]}.RandomFrame
    UVs=@n9
   ) {materialFunction=&MF_SubUV}.RGB
   B=TextureSample(UVs=Add(A=@n10, B=@n11)) {samplerType=SAMPLERTYPE_Normal, texture=/Game/Packages/VistaAssets/Stars/NewCloudImage_n}.RGB
  )
  MaterialFunctionCall(Emissive=Multiply(A=ParticleColor().RGB, B=ScalarParameter() {defaultValue=0.02, parameterName=EmissiveMult}) {constB=0.02}) {materialFunction=&MF_FX_EmissiveAdjustments}
  n12 = Desaturation(
   None=TextureSampleParameterCube!(UVs=CameraVectorWS(), Apply View MipBias) {parameterName=Cubemap, texture=/Game/CharClasses/Technician/Gadgets/BASE/Effects/T_CrystalColor_Tile}.RGB
   Fraction=ScalarParameter() {defaultValue=1, parameterName=Desat_Fraction}
  )
  n13 = MaterialFunctionCall(
   U_Images=@n1
   V_Images=@n2
   SubUV Texture RGB=TextureObject() {texture=&T_PoisonSplashShapes_6X6_Packed}
   Manual Time Control=@n7
   Time=DynamicParameter() {defaultValue=<rgba 0 0 0 0.5>, paramNames [UVDistortScale, UvTxtureScale, RandomFrame, Erode]}.RandomFrame
   UVs=@n9
  ) {materialFunction=&MF_SubUV}
  n14 = ComponentMask(None=@n13.RGB) {g=true}
  n15 = ParticleColor()
  MaterialFunctionCall(DiffuseColor=Multiply(A=LinearInterpolate(A=@n12, B=@n14), B=@n15.RGB), SpecularColor=Constant() {r=1}) {materialFunction=&ConvertFromDiffSpec}
  Multiply(B=@n15.RGB)
  n16 = DynamicParameter() {defaultValue=<rgba 0 0 0 0.5>, paramNames [UVDistortScale, UvTxtureScale, RandomFrame, Erode]}
  Multiply(
   A=Multiply(
    A=@n13.A
    B=MaterialFunctionCall!(Dynamic Param=@n13.A, Alpha=@n16.Erode) {materialFunction=/Game/Effects/Fort_Effects/Materials/_Functions/Erode_Alpha}.Result
   )
   B=@n15.A
  )
  ScalarParameter() {defaultValue=0.1, parameterName=Roughness}
 ]
}
