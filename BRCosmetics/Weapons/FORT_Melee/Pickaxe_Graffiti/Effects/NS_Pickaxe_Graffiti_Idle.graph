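# Idle Niagara effect for the Graffiti pickaxe cosmetic, sparsely shedding small paint drips from a single sprite emitter.
# Drips spawn at random offsets from the system, inherit slight velocity, then fall under downward acceleration while curves reshape and fade them.
# A random index selects pink, green or yellow color curves, and dynamic material parameters add variation.
# Front-end modifiers make drips more frequent but smaller outside gameplay.
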NiagaraSystem /BRCosmetics/Weapons/FORT_Melee/Pickaxe_Graffiti/Effects/NS_Pickaxe_Graffiti_Idle
dependencies {
 usedDataInterfaces [
  &NiagaraDataInterfaceCurve
  &NiagaraDataInterfaceVector2DCurve
  &NiagaraDataInterfaceVectorCurve
 ]
 usedDynamicInputs []
 usedModules [&SystemState]
}
emitters {Paint_Drips {
  emitterUpdateScript [
   &EmitterState()
   &SpawnRate(
    Spawn Probability=&FN_FrontEndFloatModifier(Gameplay=0.3, Front End=0.5)
    SpawnRate=6
    Use Spawn Probability=true
   )
  ]
  particleSpawnScript [
   &InitializeParticle(
    Lifetime=&RandomRangeFloat(Minimum=0.6, Maximum=0.8, Evaluation Type=::NewEnumerator1)
    Sprite Size=&Multiply_Vector2D_ByFloat(
     Float=&FN_FrontEndFloatModifier(Gameplay=1, Front End=0.6)
     Vector2D=&RandomRangeVector2D(Minimum=<1 5>, Maximum=<2 18>)
    )
    Sprite Size Mode=::NewEnumerator4
    UsePositionOffset=false
    Write Material Random=true
   )
   &SystemLocation(
    MaskSpawnGroup=false
    Offset=&RandomRangeVector(Minimum=<0 0 50>, Maximum=<5 25 60>, Evaluation Type=::NewEnumerator1)
   )
   &AddVelocity(Velocity=&RandomRangeVector(Minimum=<-10 -10 -10>, Maximum=<10 10 10>, Evaluation Type=::NewEnumerator1))
   &DynamicMaterialParameters(
    Index 0 Param 2=&RandomRangeFloat(Minimum=0.5, Maximum=1, Evaluation Type=::NewEnumerator1)
    Index 0 Param 3=&RandomRangeFloat(Minimum=0, Maximum=1, Evaluation Type=::NewEnumerator1)
    Module.Index0Float0Write=false
    Module.Index0Float1Write=true
    Module.Index0Float2Write=true
    Module.Index0Float3Write=false
   ) {module=DynamicMaterialParameters001}
   &InheritVelocity(Inherited Velocity Amount Scale=<0.1 0.1 0.1>, Speed Limit=true, UseSpeedThreshold=false)
  ]
  particleUpdateScript [
   &ParticleState()
   &ScaleSpriteSize(
    Non-Uniform Curve Sprite Scale=DataInterface() {propertyValues {
      XCurve {defaultValue=0, keys [<at 0 0>,<at 0.6 1>,<at 0.85 0.6>,<at 1 0>]}
      YCurve {defaultValue=0, keys [<at 0 0.25>,<at 0.6 1>,<at 0.85 0>,<at 1 0>]}
     }}
    Scale Sprite Size Mode=::NewEnumerator3
   )
   &AccelerationForce(Acceleration=&RandomRangeVector(Minimum=<0 0 -600>, Maximum=<0 0 -700>))
   &SolveForcesAndVelocity(Clamp Velocity=false, Limit Acceleration=false)
   &ScaleColor(
    Scale Alpha=&FloatFromCurve(FloatCurve=DataInterface() {propertyValues {Curve {keys [{arriveTangent=-1, time=0, value=1}, <at 1 0>]}}})
    Scale RGB=&MakeVectorFromLinearColorRGB(Color=&MergeHelper_ColorSwitch(
      0=&MakeLinearColorFromVectorAndFloat(
       "Vector (RGB)"=&VectorFromCurve(VectorCurve=DataInterface() {propertyValues {XCurve {keys [<at 0 5>,<at 1 2>]}, YCurve {keys [<at 0 0.9039>,<at 1 0.7608>]}, ZCurve {keys [<at 0 4.2223>,<at 1 1.1038>]}}})
       "Float (Alpha)"=1
      )
      1=&MakeLinearColorFromVectorAndFloat(
       "Vector (RGB)"=&VectorFromCurve(VectorCurve=DataInterface() {propertyValues {XCurve {keys [<at 0 0.0021>,<at 1 0.0235>]}, YCurve {keys [<at 0 2>,<at 1 1>]}, ZCurve {keys [<at 0 1.0028>,<at 1 0.5569>]}}})
       "Float (Alpha)"=1
      )
      2=&MakeLinearColorFromVectorAndFloat(
       "Vector (RGB)"=&VectorFromCurve(VectorCurve=DataInterface() {propertyValues {XCurve {keys [<at 0 2>,<at 1 2>]}, YCurve {keys [<at 0 1.7361>,<at 1 1.7>]}, ZCurve {keys [<at 0 0.2438>,<at 1 0.25>]}}})
       "Float (Alpha)"=1
      )
      3=?
      Index=&RandomRangeInt(Minimum=0, Maximum=2)
      Num=::NewEnumerator1
     ))
    ScaleA=true
    ScaleRGB=true
   )
   &DynamicMaterialParameters(
    Index 0 Param 1=0.5
    Index 0 Param 4=&FloatFromCurve(FloatCurve=DataInterface() {propertyValues {Curve {keys [<at 0 100>,<at 0.25 5>,<at 1 0>]}}})
    Module.Index0Float0Write=true
    Module.Index0Float1Write=false
    Module.Index0Float2Write=false
    Module.Index0Float3Write=true
   )
  ]
  rendererClasses [&NiagaraSpriteRendererProperties]
  renderers [{rendererClass=&NiagaraSpriteRendererProperties}]
 }}
summary {
 systemName=NS_Pickaxe_Graffiti_Idle
 emitters [{emitterName=Paint_Drips, rendererClasses [&NiagaraSpriteRendererProperties]}]
 userVariables [
  User.bIsFrontEnd {type=NiagaraBool}
  User.bIsFrontEndPreview {type=NiagaraBool}
 ]
}
