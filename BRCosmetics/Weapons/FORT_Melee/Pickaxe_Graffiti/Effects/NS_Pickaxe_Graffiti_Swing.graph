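# Niagara swing effect for the Graffiti pickaxe, pairing an animation-driven ribbon trail with falling paint-colored sprite particles.
# Main_Trail shifts from a hot magenta flash through other hues as it fades, dimmed by time of day.
# Particles_Swing spawns sprites by distance moved at a weapon socket, splitting them into two masked groups with distinct size, motion and color curves.
# Each particle randomly picks pink, green or yellow, and front-end scenes shrink sprite size.
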NiagaraSystem /BRCosmetics/Weapons/FORT_Melee/Pickaxe_Graffiti/Effects/NS_Pickaxe_Graffiti_Swing
dependencies {
 usedDataInterfaces [
  &NiagaraDataInterfaceVectorCurve
  &NiagaraDataInterfaceCurve
  &NiagaraDataInterfaceVector2DCurve
  &NiagaraDataInterfaceSocketReader
 ]
 usedDynamicInputs []
 usedModules [
  &self:SystemSpawnScript.SystemScriptSource.SystemScriptGraph.NiagaraNodeAssignment_0.SetVariables_#1
  &SystemState
 ]
}
emitters {
 Main_Trail {
  emitterSpawnScript [&FN_RibbonIDOffsetUtility()]
  emitterUpdateScript [
   &EmitterState()
   &FN_SpawnAnimTrail()
  ]
  particleSpawnScript [
   &FN_InitializeAnimTrailBoneIndecies()
   &FN_InitializeAnimTrail(AnimTrailWidth=3, Lifetime=0.3)
   &AddVelocity(Velocity=&RandomRangeVector(Minimum=<-50 0 0>, Maximum=<50 0 0>, Evaluation Type=::NewEnumerator1))
  ]
  particleUpdateScript [
   &SolveForcesAndVelocity(Clamp Velocity=false, Limit Acceleration=false)
   &FN_UpdateAnimTrail() {module=FN_UpdateAnimTrail001}
   &ScaleColor(
    Scale Alpha=&FloatFromCurve(FloatCurve=DataInterface() {propertyValues {Curve {keys [{arriveTangent=-18.2375, interpMode=RCIM_Cubic, leaveTangent=-18.2375, tangentMode=RCTM_User, time=0.3, value=5}, <at 0.9 0>]}}})
    Scale RGB=&VectorFromCurve(
     VectorCurve=DataInterface() {propertyValues {
       XCurve {keys [
         <at 0 150 RCIM_Cubic>
         {arriveTangent=-999.8954, interpMode=RCIM_Cubic, leaveTangent=-999.8954, time=0.1, value=25}
         {arriveTangent=-60.2593, interpMode=RCIM_Cubic, leaveTangent=-60.2593, time=0.15, value=0.0157}
         <at 0.5 0.8963>
         <at 0.7 1>
        ]}
       YCurve {keys [
         <at 0 0 RCIM_Cubic>
         {arriveTangent=6.6667, interpMode=RCIM_Cubic, leaveTangent=6.6667, time=0.1, value=1.5553}
         {arriveTangent=-1.9486, interpMode=RCIM_Cubic, leaveTangent=-1.9486, time=0.15, value=1}
         <at 0.5 0.7758>
         <at 0.7 0.1356>
        ]}
       ZCurve {keys [
         <at 0 110.4537 RCIM_Cubic>
         {arriveTangent=-729.691, interpMode=RCIM_Cubic, leaveTangent=-729.691, time=0.1, value=17.5549}
         {arriveTangent=-43.6136, interpMode=RCIM_Cubic, leaveTangent=-43.6136, time=0.15, value=1}
         <at 0.5 0.1095>
         <at 0.7 0.807>
        ]}
      }}
     Scale Curve=&Multiply_VectorByFloat(
      Vector=<1 1 1>
      Float=&FNTimeOfDayScalar(Night Scale=0.7, Morning Scale=0.85, Evening Scale=0.7)
     )
    )
    ScaleA=true
    ScaleRGB=true
   )
   &DynamicMaterialParameters(
    Index 0 Param 1=&FloatFromCurve(FloatCurve=DataInterface() {propertyValues {Curve {keys [
         <at 0 1 RCIM_Cubic>
         {arriveTangent=-1, interpMode=RCIM_Cubic, leaveTangent=-1, time=0.7622, value=0.5}
         {arriveTangent=-6.9586, interpMode=RCIM_Cubic, leaveTangent=-6.9586, tangentMode=RCTM_User, time=1, value=0}
        ]}}})
    Index 0 Param 2=&FloatFromCurve(FloatCurve=DataInterface() {propertyValues {Curve {keys [<at 0 0>,<at 0.25 0.2>]}}})
    Index 0 Param 3=0.5
    Index 0 Param 4=1
    Module.Index0Float0Write=true
    Module.Index0Float1Write=true
    Module.Index0Float2Write=true
    Module.Index0Float3Write=true
   )
  ]
  rendererClasses [&NiagaraRibbonRendererProperties]
  renderers [{rendererClass=&NiagaraRibbonRendererProperties}]
 }
 Particles_Swing {
  emitterUpdateScript [
   &EmitterState(Life Cycle Mode=::NewEnumerator1, Loop Behavior=::NewEnumerator1, Loop Duration=0.25, UseLoopDelay=true)
   &SpawnPerUnit(
    Max Movement Threshold=2500
    Module.Use Max Movement Threshold=true
    Module.Use Movement Tolerance=false
    Spawn Spacing=30
    Use Spawn Probability=false
   )
  ]
  particleSpawnScript [
   &FN_SetSpawnGroup(Number of Spawn Groups=2)
   &InitializeParticle(
    Lifetime=&MaskFloat(
     Mask Source=0
     Mask Parameter=?
     Mask=0
     Masked Float=&RandomRangeFloat(Minimum=0.35, Maximum=0.5)
     Passthrough Float=&RandomRangeFloat(Minimum=0.4, Maximum=0.65)
    )
    Sprite Rotation Mode=::NewEnumerator2
    Sprite Size=&Multiply_Vector2D_ByFloat(
     Float=&FN_FrontEndFloatModifier(Gameplay=1, Front End=0.6)
     Vector2D=/Niagara/DynamicInputs/Helpers/MaskVector2d(
      Mask Source:NiagaraInt32=0
      Mask Parameter:NiagaraInt32=?
      Mask:NiagaraInt32=0
      Masked Vector2D:Vector2f=&RandomRangeVector2D(Minimum=<1 8>, Maximum=<2 25>)
      Passthrough Vector2D:Vector2f=&RandomRangeVector2D(Minimum=<2 12>, Maximum=<3 20>)
     )
    )
    Sprite Size Mode=::NewEnumerator4
    UsePositionOffset=false
    Write Material Random=true
   )
   &SocketLocation(
    Interpolate Socket Sampling=true
    SamplePosOffset=false
    Socket Index=1
    Socket Reader=$System.Socket Reader
    Socket Sampling Mode=::NewEnumerator3
    VelocitySpeedLimit=false
   )
   &DynamicMaterialParameters(
    Index 0 Param 1=&RandomRangeFloat(Minimum=0.2, Maximum=0.3, Evaluation Type=::NewEnumerator1)
    Index 0 Param 2=&RandomRangeFloat(Minimum=0.6, Maximum=1, Evaluation Type=::NewEnumerator1)
    Index 0 Param 3=&RandomRangeFloat(Minimum=0, Maximum=1, Evaluation Type=::NewEnumerator1)
    Index 2 Param 2=&RandomRangeFloat(Minimum=0.5, Maximum=1, Evaluation Type=::NewEnumerator1)
    Index 2 Param 3=&RandomRangeFloat(Minimum=0, Maximum=1)
    Module.Index0Float0Write=true
    Module.Index0Float1Write=true
    Module.Index0Float2Write=true
    Module.Index0Float3Write=false
    Module.Index2Float0Write=false
    Module.Index2Float1Write=true
    Module.Index2Float2Write=true
    Module.Index2Float3Write=false
    Write Parameter Index 2=::NewEnumerator1
   )
   &AddVelocity(Velocity=&MaskVector(
     Mask Source=0
     Mask Parameter=?
     Mask=0
     Masked Vector=&RandomRangeVector(Minimum=<0 0 50>, Maximum=<0 0 -50>, Evaluation Type=::NewEnumerator1)
     Passthrough Vector=&RandomRangeVector(Minimum=<-35 -35 -35>, Maximum=<35 35 35>, Evaluation Type=::NewEnumerator1)
    ))
  ]
  particleUpdateScript [
   &ParticleState()
   &ScaleSpriteSize(
    Scale Factor=&MaskVector2d(
     Mask Source:NiagaraInt32=0
     Mask Parameter:NiagaraInt32=?
     Mask:NiagaraInt32=0
     Masked Vector2D:Vector2f=&Vector2DFromCurve(
      Vector2Curve=DataInterface() {propertyValues {
        XCurve {defaultValue=0, keys [<at 0 0>,<at 0.5 1.25>,<at 0.85 0.6>,<at 1 0>]}
        YCurve {defaultValue=0, keys [<at 0 0.1>,<at 0.5 1>,<at 0.85 0>,<at 1 0>]}
       }}
      ScaleCurve=<1 1>
     )
     Passthrough Vector2D:Vector2f=&Vector2DFromCurve(
      Vector2Curve=DataInterface() {propertyValues {
        XCurve {defaultValue=0, keys [<at 0 0>,<at 0.5 1>,<at 0.85 0.6>,<at 1 0>]}
        YCurve {defaultValue=0.1, keys [<at 0 0.25>,<at 0.5 1>,<at 0.85 0>,<at 1 0>]}
       }}
      ScaleCurve=<1 1>
     )
    )
    Scale Sprite Size Mode=::NewEnumerator2
   )
   &ScaleColor(
    Scale Alpha=&FloatFromCurve(FloatCurve=DataInterface() {propertyValues {Curve {keys [{arriveTangent=-1, time=0, value=1}, <at 1 0>]}}})
    Scale RGB=&MaskVector(
     Mask Source=0
     Mask Parameter=?
     Mask=0
     Masked Vector=&MergeHelper_VectorSwitch(
      0=&VectorFromCurve(VectorCurve=DataInterface() {propertyValues {XCurve {keys [<at 0 2>,<at 1 2>]}, YCurve {keys [<at 0 0.3616>,<at 1 0.7608>]}, ZCurve {keys [<at 0 1.6889>,<at 1 1.1038>]}}})
      1=&VectorFromCurve(VectorCurve=DataInterface() {propertyValues {XCurve {keys [<at 0 0.0082>,<at 1 0.05>]}, YCurve {keys [<at 0 2>,<at 1 2>]}, ZCurve {keys [<at 0 0.9301>,<at 1 1>]}}})
      2=&VectorFromCurve(VectorCurve=DataInterface() {propertyValues {XCurve {keys [<at 0 2>,<at 1 2>]}, YCurve {keys [<at 0 1.754>,<at 1 1.7>]}, ZCurve {keys [<at 0 0.206>,<at 1 0.2>]}}})
      3=?
      Index=&RandomRangeInt(Minimum=0, Maximum=2)
      Num=::NewEnumerator1
     )
     Passthrough Vector=&MergeHelper_VectorSwitch(
      0=&VectorFromCurve(VectorCurve=DataInterface() {propertyValues {XCurve {keys [<at 0 5>,<at 1 2>]}, YCurve {keys [<at 0 0.9039>,<at 1 0.7608>]}, ZCurve {keys [<at 0 4.2223>,<at 1 1.1038>]}}})
      1=&VectorFromCurve(VectorCurve=DataInterface() {propertyValues {XCurve {keys [<at 0 0.0021>,<at 1 0.0235>]}, YCurve {keys [<at 0 2>,<at 1 1>]}, ZCurve {keys [<at 0 1.0028>,<at 1 0.5569>]}}})
      2=&VectorFromCurve(VectorCurve=DataInterface() {propertyValues {XCurve {keys [<at 0 2>,<at 1 2>]}, YCurve {keys [<at 0 1.7361>,<at 1 1.7>]}, ZCurve {keys [<at 0 0.2438>,<at 1 0.25>]}}})
      3=?
      Index=&RandomRangeInt(Minimum=0, Maximum=2)
      Num=::NewEnumerator1
     )
    )
    ScaleA=true
    ScaleRGB=true
   )
   &AccelerationForce(Acceleration=&MaskVector(
     Mask Source=0
     Mask Parameter=?
     Mask=0
     Masked Vector=&RandomRangeVector(Minimum=<0 0 -1200>, Maximum=<0 0 -1600>)
     Passthrough Vector=&RandomRangeVector(Minimum=<0 0 -800>, Maximum=<0 0 -900>)
    ))
   &SolveForcesAndVelocity(Clamp Velocity=false, Limit Acceleration=false)
   &DynamicMaterialParameters(
    Index 0 Param 4=&FloatFromCurve(FloatCurve=DataInterface() {propertyValues {Curve {keys [<at 0 25>,<at 0.25 25>,<at 1 0>]}}})
    Index 2 Param 1=0.5
    Index 2 Param 4=&FloatFromCurve(FloatCurve=DataInterface() {propertyValues {Curve {keys [<at 0 100>,<at 0.25 5>,<at 1 0>]}}})
    Module.Index0Float0Write=false
    Module.Index0Float1Write=false
    Module.Index0Float2Write=false
    Module.Index0Float3Write=true
    Module.Index2Float0Write=true
    Module.Index2Float1Write=false
    Module.Index2Float2Write=false
    Module.Index2Float3Write=true
    Write Parameter Index 2=::NewEnumerator1
   ) {module=DynamicMaterialParameters001}
  ]
  rendererClasses [&NiagaraSpriteRendererProperties]
  renderers [
   {rendererClass=&NiagaraSpriteRendererProperties}
   {rendererClass=&NiagaraSpriteRendererProperties, rendererIndex=1}
  ]
 }
}
summary {
 systemName=NS_Pickaxe_Graffiti_Swing
 emitters [
  {emitterName=Main_Trail, rendererClasses [&NiagaraRibbonRendererProperties]}
  {emitterName=Particles_Swing, rendererClasses [&NiagaraSpriteRendererProperties]}
 ]
 userVariables [
  User.bIsFrontEnd {type=NiagaraBool}
  User.bIsFrontEndPreview {type=NiagaraBool}
  User.RibbonIDOffset {type=NiagaraInt32}
 ]
}
