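# Physics asset for the Graffiti pickaxe skeletal mesh, a two-handed melee cosmetic with a dynamic part named dyn_rolla, apparently a roller.
# A single long capsule covers the handle bone, while two capsules form the dyn_rolla body.
# One constraint joins dyn_rolla to the handle with both swings locked and twist free, so the part can only rotate about the twist axis.
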PhysicsAsset /BRCosmetics/Weapons/FORT_Melee/Pickaxe_Graffiti/Meshes/SK_Pickaxe_Graffiti_PhysicsAsset
bodies {
 Melee_TwoHanded_Handle [{shapeType=Capsule, length=100, radius=6, center=<0 0 13.752>, rotation=<rot 0 90 0>}]
 dyn_rolla [
  {shapeType=Capsule, length=27.7431, radius=5.1628, center=<0.5292 21.0233 -0.2666>, rotation=<rot -0.0003 -1.5688 -89.4855>}
  {shapeType=Capsule, length=2.5, radius=2.5, shapeName=dyn_rolla_capsule, center=<0.4456 -2.1416 0.2953>, rotation=<rot -0.0002 0 89.9994>}
 ]
}
constraints [{
  bone1Name=dyn_rolla
  bone2Name=Melee_TwoHanded_Handle
  swing1LimitDegrees=45
  swing1Motion=Locked
  swing2LimitDegrees=45
  swing2Motion=Locked
  twistLimitDegrees=45
  twistMotion=Free
 }]
